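# Animation Blueprint for the YuzuCrank male medium body, copying the pose from its attached parent mesh and refining it with a Control Rig.
# The YuzuCrank Control Rig receives Emote On and Emote Off weights alongside the copied pose.
# Each update, the event graph interpolates those two weights toward 0 or 1 based on bIsPlayingEmote, with different rates smoothing the switch in and out of emotes.
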AnimBlueprint /BRCosmetics/Characters/Player/Male/Medium/Bodies/M_MED_YuzuCrank/Meshes/M_MED_YuzuCrank_AnimBP
target_skeleton=&/M_MED_YuzuCrank_Skeleton
graphs [
 AnimGraph {nodes [
   n1 = Control Rig!(
    Source=Copy Pose From Mesh() {bUseAttachedParent=true}
    Alpha
    Emote On:double=Get Emote on() -> (Emote on:double)
    Emote Off:double=Get Emote off() -> (Emote off:double)
   ) -> (Pose) {
    alphaBoolBlend {blendOption=Linear}
    controlRigAssetReference {blueprintRigClass=&/M_MED_YuzuCrank_CtrlRig.M_MED_YuzuCrank_CtrlRig_C}
   }
   Output Pose(Result=@n1) {layerGroup=None}
  ]}
 EventGraph {nodes [
   n1 = Event BlueprintUpdateAnimation()
   > Set Delta Time X(Delta Time X=@n1.DeltaTimeX)
   > n2 = Sequence()
   > n3 = Branch(Condition=Get bIsPlayingEmote())
   n4 = float / float(A=Get Delta Time X(), B=2)
   > Set Emote on(Emote on:double=FInterpTo(Current=Get Emote on(), Target=1, DeltaTime=@n4, InterpSpeed=20)) -> (Output_Get:double)
   Set Emote on(execute=@n3.else, Emote on=FInterpTo(Current=Get Emote on(), Target=0, DeltaTime=@n4, InterpSpeed=20))
   n5 = Branch(execute=@n2.then_1, Condition=NOT Boolean(A=Get bIsPlayingEmote()))
   n6 = Get Delta Time X()
   > Set Emote off(Emote off:double=FInterpTo(Current=Get Emote off(), Target=1, DeltaTime=@n6, InterpSpeed=12)) -> (Output_Get:double)
   Set Emote off(execute=@n5.else, Emote off=FInterpTo(Current=Get Emote off(), Target=0, DeltaTime=@n6, InterpSpeed=20))
   float / float(A=@n6, B=2)
  ]}
]
variables [
 Delta Time X {type=number, default=0.0541}
 Emote off {type=number, default=1}
 Emote on {type=number}
]
